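# Unlit translucent Niagara material that draws procedural toon smoke puffs.
# Randomised polar petals and two panned noise layers distort UVs into a lumpy sphere mask, faded by particle alpha and depth.
# Normals rebuilt from the distorted disc are lit by sky-atmosphere light for contrasted two-tone shading, with procedural shadow petals.
# Dynamic parameters expose shadow colour, time, petal count and WPO multiplier; platform switches drop depth fade and shadow petals on some targets.
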Material /WildEstateTech/FX/ToonSmoke/M_ToonSmokePuff
graph {
 properties {
  bAutomaticallySetUsageInEditor=false
  bUsedWithNiagaraMeshParticles=true
  bUsedWithNiagaraRibbons=true
  bUsedWithNiagaraSprites=true
  blendMode=BLEND_Translucent
  shadingModel=MSM_Unlit
  usedShadingModels=Unlit
 }
 nodes [
  DynamicParameter() {parameterIndex=2, defaultValue=<rgba 0 0 0 0>, paramNames [Na,Na,Na,Na]}
  DynamicParameter() {parameterIndex=3, defaultValue=<rgba 0 0 0 0>, paramNames [Na,Na,Na,Na]}
  n1 = ConstantBiasScale(None=TextureCoordinate()) {bias=-0.5, scale=2}
  n2 = Arctangent2Fast(Y=ComponentMask(None=@n1) {g=true}, X=ComponentMask(None=@n1) {r=true})
  n3 = Frac(None=Divide(A=ComponentMask(None=@n2) {r=true}, B=Constant() {r=6.2832}))
  n4 = LinearInterpolate(
   A=ScalarParameter() {defaultValue=10, group=Petal, parameterName=Pedal COunt}
   B=NamedRerouteUsage()
   Alpha=ScalarParameter() {group=Petal, parameterName=Use Dyam Petal count}
  )
  n5 = Multiply(A=@n3, B=@n4) {constB=10}
  n6 = ConstantBiasScale(None=AppendVector(A=Frac(None=@n5), B=Constant() {r=0.5})) {bias=-0.5, scale=2}
  n7 = OneMinus(None=DotProduct(A=@n6, B=@n6))
  n8 = ScalarParameter() {defaultValue=1, group=Petal, parameterName=Petal Exp low}
  n9 = ScalarParameter() {defaultValue=3, group=Petal, parameterName=Petal Exp High}
  n10 = Multiply(A=@n3, B=@n4)
  n11 = VectorParameter() {
   group=Petal
   parameterName=Petal RandOptions
   channelNames {a=Na, b=Variation, g=SeedY, r=Seed X}
   defaultValue=<rgba 12.9898 78.233 43758.5469 0>
  }
  n12 = MaterialFunctionCall!(UVs=Divide(A=Floor(None=@n10), B=@n4), Seed Direction=ComponentMask(None=@n11) {g=true, r=true}, Variation Mul=@n11.Variation) {materialFunction=&/MF_Random2}
  n13 = Multiply(
   A=Power(Base=SquareRoot(None=@n7), Exp=LinearInterpolate(A=@n8, B=@n9, Alpha=@n12.X)) {constExponent=3}
   B=LinearInterpolate(
    A=ScalarParameter() {defaultValue=0.035, group=Petal, parameterName=Petal SizeLow}
    B=ScalarParameter() {defaultValue=0.05, group=Petal, parameterName=Petal Size High}
    Alpha=@n12.Y
   )
  )
  LinearInterpolate(
   A=@n13
   B=@n13
   Alpha=MaterialFunctionCall(Speed=Constant() {r=3}) {materialFunction=&DebugTimeSine}.Result
  )
  n14 = VectorParameter() {
   group=Opac
   parameterName=Depth Fade Options
   channelNames {a=Pixel depth Offset, b=Camera Fade Offset, g=Camera Fade Length, r=Depth Fade Value}
  }
  LinearInterpolate(
   B=@n14.Pixel depth Offset
   Alpha=MaterialFunctionCall(Alpha Threshold=Constant() {r=1}) {materialFunction=&DitherTemporalAA}.Result
  )
  MaterialFunctionCall() {materialFunction=&DebugTimeSine}
  n15 = ConstantBiasScale(None=TextureCoordinate()) {bias=-0.5, scale=2}
  n16 = TextureCoordinate()
  n17 = Add(A=TextureCoordinate(), B=NamedRerouteUsage())
  n18 = VectorParameter() {
   group=Distortion
   parameterName=Distortion 1 Options
   channelNames {a=subtract Offset, b=Srength, g=Tile Y, r=tile X}
   defaultValue=<rgba 0.5 0.5 0.1 0.5>
  }
  n19 = ComponentMask(None=@n18) {g=true, r=true}
  n20 = ScalarParameter() {defaultValue=0.1, group=Distortion, parameterName=Distort Speed 1 X}
  n21 = ScalarParameter() {defaultValue=0.1, group=Distortion, parameterName=Distort Speed 1 Y}
  n22 = TextureSample(
   UVs=Panner(Coordinate=Multiply(A=@n17, B=@n19) {constB=0.7}, Time=NamedRerouteUsage(), Speed=AppendVector(A=@n20, B=@n21))
   Tex=NamedRerouteUsage()
  ) {samplerType=SAMPLERTYPE_LinearColor, texture=&HighContrastNoise}
  n23 = Multiply(A=Subtract(A=ComponentMask(None=@n22.RGB) {g=true, r=true}, B=@n18.subtract Offset) {constB=0.5}, B=@n18.Srength) {constB=0.1}
  n24 = Add(A=TextureCoordinate(), B=NamedRerouteUsage())
  n25 = VectorParameter() {
   group=Distortion
   parameterName=Distortion 2 Options
   channelNames {a=subtract Offset, b=Srength, g=Tile Y, r=tile X}
   defaultValue=<rgba 0.25 0.25 0.25 0.5>
  }
  n26 = ComponentMask(None=@n25) {g=true, r=true}
  n27 = Panner(
   Coordinate=MaterialFunctionCall(XY=Multiply(A=@n24, B=@n26) {constB=0.7}) {materialFunction=&Swizzle}.YX
   Time=NamedRerouteUsage()
   Speed=AppendVector(
    A=ScalarParameter() {defaultValue=0.1, group=Distortion, parameterName=Distort Speed 2 X}
    B=ScalarParameter() {defaultValue=-0.1, group=Distortion, parameterName=Distort Speed 2 Y}
   )
  )
  n28 = NamedRerouteUsage()
  n29 = Subtract(
   A=ComponentMask(None=TextureSample(UVs=@n27, Tex=@n28) {samplerType=SAMPLERTYPE_LinearColor, texture=&HighContrastNoise}.RGB) {b=true, g=true}
   B=@n25.subtract Offset
  ) {constB=0.5}
  n30 = Subtract(A=TextureCoordinate()) {constB=0.5}
  n31 = LinearInterpolate(Alpha=ScalarParameter() {group=Petal, parameterName=invert Petal, sliderMax=1}) {constA=-2, constB=2}
  n32 = Add(A=Add(A=Add(A=@n16, B=@n23), B=Multiply(A=@n29, B=@n25.Srength) {constB=0.1}), B=Multiply(A=Multiply(A=@n30, B=@n31) {constB=-2}, B=@n13))
  n33 = ScalarParameter() {defaultValue=0.65, group=Distortion, parameterName=Size}
  n34 = VectorParameter() {group=Distortion, parameterName=Center Distort Options, channelNames {a=Contrast, b=add, g=mul, r=pow}, defaultValue=<rgba 1 1 0 0>}
  n35 = Power(Base=Distance(A=@n15, B=Constant()), Exp=@n34.pow)
  n36 = MaterialFunctionCall(In=Add(A=Multiply(A=@n35, B=@n34.mul), B=@n34.add), Contrast=@n34.Contrast) {materialFunction=&CheapContrast}
  n37 = LinearInterpolate(
   A=@n15
   B=Multiply(A=ConstantBiasScale(None=@n32) {bias=-0.5, scale=2}, B=Divide(B=@n33))
   Alpha=LinearInterpolate(
    A=ScalarParameter() {defaultValue=1, group=Distortion, parameterName="inner  Distort"}
    B=ScalarParameter() {defaultValue=1, group=Distortion, parameterName=Outer Distort}
    Alpha=Saturate(None=@n36.Result)
   )
  )
  n38 = SphereMask(A=@n37, B=Constant(), Radius=Constant() {r=1}, Hardness=ScalarParameter() {defaultValue=0.8, group=Opac, parameterName=Hardness}) {attenuationRadius=1}
  n39 = ScalarParameter() {defaultValue=2, group=Opac, parameterName=SPhere mask EXP}
  n40 = Multiply(
   A=Multiply(A=Power(Base=@n38, Exp=@n39), B=ParticleColor().A)
   B=MaterialFunctionCall(Fade Length=@n14.Camera Fade Length, Fade Offset=@n14.Camera Fade Length) {materialFunction=&CameraDepthFade}.Result
  )
  MaterialFunctionCall(Alpha Threshold=@n40) {materialFunction=&DitherTemporalAA}
  MaterialFunctionCall() {materialFunction=&LinearGradient}
  MaterialFunctionCall() {materialFunction=&DebugTimeSine}
  n41 = NamedRerouteUsage()
  n42 = ParticleColor()
  n43 = Normalize(VectorInput=SkyAtmosphereLightDirection())
  n44 = DotProduct(A=@n37, B=@n37)
  n45 = AppendVector(A=@n37, B=SquareRoot(None=OneMinus(None=@n44)))
  n46 = Normalize(VectorInput=Transform(None=@n45))
  n47 = ScalarParameter() {defaultValue=-0.5, group=Shadow, parameterName=Shadow Offset}
  n48 = Add(A=Add(A=DotProduct(A=@n43, B=@n46), B=Multiply(A=@n47) {constB=1.05})) {constB=0.5}
  n49 = NamedRerouteUsage()
  n50 = MaterialFunctionCall(In=Saturate(None=Add(A=@n48, B=@n49) {constB=0}), Contrast=ScalarParameter() {defaultValue=5, group=Shadow, parameterName=Shadow Contrast}) {materialFunction=&CheapContrast}
  n51 = DepthFade(Opacity=@n40, FadeDistance=@n14.Depth Fade Value)
  n52 = MaterialFunctionCall("Xb1,PS4"=@n51, "PC-High&&Epic,XSX,PS5"=@n51, "PC-Perf,Android"=@n40, Switch=@n40, "PC-Low(NotPerf)&&Medium"=@n51) {materialFunction=&PlatformSwitch_1D}
  Multiply(
   A=MaterialFunctionCall(Emissive=LinearInterpolate(A=@n41, B=@n42.RGB, Alpha=@n50.Result)) {materialFunction=&MF_FX_EmissiveAdjustments}.Result
   B=@n52.Result
  )
  n53 = Multiply(
   A=ConstantBiasScale(None=TextureCoordinate()) {bias=-0.5, scale=2}
   B=Divide(B=ScalarParameter() {defaultValue=1, group=WPO, parameterName=WPO Sphere size})
  )
  n54 = SquareRoot(None=OneMinus(None=DotProduct(A=@n53, B=@n53)))
  Multiply(
   A=Multiply(A=Saturate(None=@n54), B=VertexNormalWS())
   B=Multiply(A=ScalarParameter() {group=WPO, parameterName=WPO Str}, B=NamedRerouteUsage())
  )
  n55 = SceneTexture(UVs=ScreenPosition().ViewportUV) {sceneTextureId=PPI_SeparateTranslucency}
  Multiply(A=@n52.Result, B=OneMinus(None=ComponentMask(None=@n55.Color) {a=true}))
  n56 = DynamicParameter() {defaultValue=<rgba 0.5 0.675 1 1>, paramNames [Shadow Color R, Shadow Color G, Shadow Color B, Time]}
  NamedRerouteDeclaration(None=@n56.RGB) {name=DP0 RGB Shadow Color, nodeColor=<rgba 1 0 0 1>}
  NamedRerouteDeclaration(None=@n56.Time) {name=DP0 A Time, nodeColor=<rgba 1 0 0 1>}
  NamedRerouteDeclaration(None=@n13) {name=Petal Offset, nodeColor=<rgba 0 1 0.3906 1>}
  n57 = DynamicParameter() {parameterIndex=1, defaultValue=<rgba 1 1 0 0>, paramNames [Petal Count, WPO Mul, Na, Na]}
  NamedRerouteDeclaration(None=@n57.Petal Count) {name=DP1 R PetalCount, nodeColor=<rgba 1 0 0 1>}
  NamedRerouteDeclaration(None=@n57.WPO Mul) {name=DP1 G WPO Mul, nodeColor=<rgba 0 1 0 1>}
  NamedRerouteDeclaration(None=@n45) {name=Proc Normals, nodeColor=<rgba 0.25 0.25 1 1>}
  n58 = Constant()
  n59 = MaterialFunctionCall(Float3=NamedRerouteUsage()) {materialFunction=&BreakOutFloat3Components}
  n60 = Multiply(A=@n59.B) {constB=-1}
  n61 = SkyAtmosphereLightDirection()
  n62 = CrossProduct(A=@n61, B=Constant3Vector() {constant=<rgba 0 0 1 0>})
  n63 = MaterialFunctionCall!(Vector1=@n62, Vector2=CrossProduct(A=@n61, B=@n62), Vector3=@n61) {materialFunction=/Engine/Functions/Engine_MaterialFunctions02/Math/MakeVectorsOrthogonal}
  n64 = MaterialFunctionCall(Float3=MaterialFunctionCall!(
    VectorToTransform=MaterialFunctionCall(X=@n60, Y=@n59.R, Z=@n59.R) {materialFunction=&MakeFloat3}.Result
    BasisX=@n63.Vector1
    BasisY=@n63.Vector1
    BasisZ=@n63.Vector1
   ) {materialFunction=/Engine/Functions/Engine_MaterialFunctions02/Math/InverseTransformMatrix}.Transformed Vector 3) {materialFunction=&BreakOutFloat3Components}
  n65 = Sign(None=@n64.R)
  n66 = MaterialFunctionCall(X=Multiply(A=Saturate(None=@n65)) {constB=0.5}) {materialFunction=&MakeFloat2}
  n67 = Multiply(A=@n64.B, B=@n64.B)
  n68 = Divide(A=@n64.G, B=SquareRoot(None=OneMinus(None=@n67)))
  n69 = ArcsineFast!(None=MaterialFunctionCall(X=Multiply(A=@n68, B=@n65), Y=@n64.B) {materialFunction=&MakeFloat2}.Result)
  n70 = MaterialFunctionCall() {materialFunction=&Pi}
  n71 = Multiply(A=Constant2Vector() {g=1, r=0.5}, B=Add(A=Divide(A=@n69, B=@n70.Result)) {constB=0.5})
  n72 = ScalarParameter() {defaultValue=15, group=Shadow, parameterName=Shadow Petal Count}
  n73 = Multiply(
   A=MaterialFunctionCall(Float2=Add(A=@n66.Result, B=@n71) {desc=ProcSphere UVs}) {materialFunction=&BreakOutFloat2Components}.R
   B=@n72
  )
  n74 = ConstantBiasScale(None=Frac(None=@n73)) {bias=-0.5, scale=2}
  n75 = DotProduct(A=@n74, B=@n74)
  n76 = Saturate(None=SquareRoot(None=OneMinus(None=@n75)))
  n77 = MaterialFunctionCall(UVs=Divide(A=Floor(None=@n73), B=@n72), Variation Mul=ScalarParameter() {defaultValue=138265.375, parameterName=Shadow Petal Variation}) {materialFunction=&MF_Random2}
  n78 = LinearInterpolate(
   A=ScalarParameter() {defaultValue=1, group=Shadow, parameterName=Shadow Petal Exp low}
   B=ScalarParameter() {defaultValue=5, group=Shadow, parameterName=Shadow Petal Exp High}
   Alpha=@n77.X
  )
  n79 = ScalarParameter() {defaultValue=0.5, group=Shadow, parameterName=Shadow Petal Offset Low}
  n80 = ScalarParameter() {defaultValue=0.5, group=Shadow, parameterName=Shadow Petal Offset High}
  n81 = Frac(None=Add(A=@n77.X, B=@n77.X))
  n82 = Multiply(
   A=Subtract(A=Power(Base=@n76, Exp=@n78) {constExponent=1}, B=LinearInterpolate(A=@n79, B=@n80, Alpha=@n81)) {constB=0.5}
   B=LinearInterpolate(
    A=ScalarParameter() {defaultValue=0.05, group=Shadow, parameterName=Shadow Petal SizeLow}
    B=ScalarParameter() {defaultValue=0.2, group=Shadow, parameterName=Shadow Petal Size High}
    Alpha=@n77.Y
   )
  )
  NamedRerouteDeclaration(None=MaterialFunctionCall("Xb1,PS4"=@n58, "PC-High&&Epic,XSX,PS5"=@n82, "PC-Perf,Android"=@n58, Switch=@n58, "PC-Low(NotPerf)&&Medium"=@n82) {materialFunction=&PlatformSwitch_1D}.Result) {name=Shadow Petals, nodeColor=<rgba 0 1 0.6953 1>}
  n83 = VectorParameter() {parameterName=Distor Distortion Tile Speed, channelNames {a=Speed Y, b=Speed X, g=Tile Y, r=Tile X}, defaultValue=<rgba 1 1 0.1 0.3>}
  n84 = Multiply(A=TextureCoordinate(), B=AppendVector(A=@n83.Tile X, B=@n83.Tile X))
  n85 = VectorParameter() {
   parameterName="Distort Distortion Offset, Center, Strnength"
   channelNames {a=Strength, b=Center, g=Offset Y, r=Offset X}
   defaultValue=<rgba -0.72 0.528 0.5 0.1>
  }
  n86 = AppendVector(A=@n85.Offset X, B=@n85.Offset X)
  n87 = TextureSample(
   UVs=Panner(Coordinate=Add(A=@n84, B=@n86), Time=NamedRerouteUsage(), Speed=AppendVector(A=@n83.Speed X, B=@n83.Speed X))
   Tex=NamedRerouteUsage()
  ) {samplerType=SAMPLERTYPE_LinearColor, texture=&HighContrastNoise}
  n88 = AppendVector(A=@n87.G, B=@n87.G)
  NamedRerouteDeclaration(None=Multiply(A=Subtract(A=@n88, B=@n85.Center), B=@n85.Strength)) {name=Distort Distortion, nodeColor=<rgba 1 0 1 1>}
  n89 = TextureObjectParameter() {
   group=Distortion
   parameterName=Distortion Texture RGB
   samplerType=SAMPLERTYPE_LinearColor
   texture=&HighContrastNoise
  }
  NamedRerouteDeclaration(None=@n89) {name=Param Distortin Texture, nodeColor=<rgba 0 0.35 1 1>}
  OneMinus()
  ParticleRandom()
  Reroute(None=@n89)
  SphereMask(A=TextureCoordinate(), B=Constant() {r=0.5}, Radius=Add(A=@n13, B=Constant() {r=0.3})) {desc=preview, hardnessPercent=0}
  Time() {desc=debug}
 ]
}
